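# Control Rig for the Ebony HoodDown face accessory part that applies corrective offsets to front straps, hood and back hair.
# Spherical pose readers on left clavicle and right upper arm weight Modify Transforms on straps and hood.
# Backbling, then Skydiving, branches select head-driven hood corrections, with hair included except when skydiving; Backbling adds a hood offset weighted by smoothed Moving.
# A left upper-arm strap reader sits outside the execution chain.
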ControlRigBlueprint /BRCosmetics/Characters/Player/Male/Medium/Bodies/M_MED_Ebony_HoodDown/Meshes/Parts/M_MED_Ebony_HoodDown_FaceAcc_CtrlRig
graphs [
 RigVMFunctionLibrary {links [], nodes []}
 RigVMModel {nodes [
   RigUnit_BeginExecution = &RigUnit_BeginExecution()
   n1 = &RigUnit_SphericalPoseReader(
    ActiveRegionScaleFactors=(PositiveWidth=1,NegativeWidth=0.5,PositiveHeight=0.5,NegativeHeight=1)
    ActiveRegionSize=0.1
    Debug=(bDrawDebug=false,bDrawLocalAxes=false,DebugScale=25,DebugSegments=20,DebugThickness=0.25)
    DriverAxis=(X=1,Y=0,Z=0)
    DriverItem=(Type=Bone,Name="clavicle_l")
    FalloffRegionScaleFactors=(PositiveWidth=0.3,NegativeWidth=0.3,PositiveHeight=0.3,NegativeHeight=0.3)
    FalloffSize=0.45
    FlipHeightScaling=false
    FlipWidthScaling=true
    OptionalParentItem=(Type=Bone)
    RotationOffset=(X=0,Y=-40,Z=20)
    ExecutePin=@RigUnit_BeginExecution.ExecutePin
   )
   > &RigUnit_ModifyTransforms(
    ItemToModify=((Item=(Type=Bone,Name="dyn_strap_fr_l_MSTR1"),Transform=(Rotation=(X=0,Y=0,Z=0,W=1),Translation=(X=1,Y=2,Z=-3),Scale3D=(X=0.5,Y=1,Z=1))),(Item=(Type=Bone),Transform=(Rotation=(X=0,Y=0,Z=0,W=1),Translation=(X=0,Y=0,Z=0),Scale3D=(X=1,Y=1,Z=1))))
    ItemToModify.0.Transform.Rotation=&RigVMFunction_MathQuaternionFromEuler(Euler=(X=0,Y=-20,Z=0), RotationOrder=XYZ).Result
    Weight=@n1.OutputParam
   )
   > n2 = &RigUnit_SphericalPoseReader(
    ActiveRegionScaleFactors=(PositiveWidth=1,NegativeWidth=1,PositiveHeight=1,NegativeHeight=1)
    ActiveRegionSize=0.1
    Debug=(bDrawDebug=false,bDraw2D=false,bDrawLocalAxes=false,DebugScale=25,DebugSegments=20,DebugThickness=0.25)
    DriverAxis=(X=-1,Y=0,Z=0)
    DriverItem=(Type=Bone,Name="upperarm_r")
    FalloffRegionScaleFactors=(PositiveWidth=0.2,NegativeWidth=0.2,PositiveHeight=0.2,NegativeHeight=0.2)
    FalloffSize=0.45
    FlipHeightScaling=false
    FlipWidthScaling=true
    OptionalParentItem=(Type=Bone,Name="spine_04")
    RotationOffset=(X=-130,Y=60,Z=10)
   )
   > &RigUnit_ModifyTransforms(
    ItemToModify=((Item=(Type=Bone,Name="dyn_strap_fr_r_MSTR1"),Transform=(Rotation=(X=0,Y=0,Z=0,W=1),Translation=(X=1,Y=1.5,Z=2),Scale3D=(X=1,Y=1,Z=1))),(Item=(Type=Bone),Transform=(Rotation=(X=0,Y=0,Z=0,W=1),Translation=(X=0,Y=0,Z=0),Scale3D=(X=1,Y=1,Z=1))))
    ItemToModify.0.Transform.Rotation=&RigVMFunction_MathQuaternionFromEuler(Euler=(X=0,Y=0,Z=0), RotationOrder=XYZ).Result
    Weight=@n2.OutputParam
   )
   > n3 = &RigUnit_SphericalPoseReader(
    ActiveRegionScaleFactors=(PositiveWidth=1,NegativeWidth=1.5,PositiveHeight=0.6,NegativeHeight=0.5)
    ActiveRegionSize=0.1
    Debug=(bDrawDebug=false,bDraw2D=false,bDrawLocalAxes=false,DebugScale=25,DebugSegments=20,DebugThickness=0.25)
    DriverAxis=(X=-1,Y=0,Z=0)
    DriverItem=(Type=Bone,Name="upperarm_r")
    FalloffRegionScaleFactors=(PositiveWidth=0.2,NegativeWidth=0.2,PositiveHeight=0.15,NegativeHeight=0.08)
    FalloffSize=0.45
    FlipHeightScaling=false
    FlipWidthScaling=true
    OptionalParentItem=(Type=Bone,Name="spine_04")
    RotationOffset=(X=-70,Y=45,Z=30)
   )
   > &RigUnit_ModifyTransforms(
    ItemToModify=((Item=(Type=Bone,Name="dyn_strap_fr_r_MSTR1"),Transform=(Rotation=(X=0,Y=0,Z=0,W=1),Translation=(X=0.5,Y=1,Z=2),Scale3D=(X=1,Y=1,Z=1))),(Item=(Type=Bone),Transform=(Rotation=(X=0,Y=0,Z=0,W=1),Translation=(X=0,Y=0,Z=0),Scale3D=(X=1,Y=1,Z=1))))
    ItemToModify.0.Transform.Rotation=&RigVMFunction_MathQuaternionFromEuler(Euler=(X=8,Y=0,Z=0), RotationOrder=XYZ).Result
    Weight=@n3.OutputParam
   )
   > n4 = &RigUnit_SphericalPoseReader(
    ActiveRegionScaleFactors=(PositiveWidth=1,NegativeWidth=2,PositiveHeight=1,NegativeHeight=1.4)
    ActiveRegionSize=0.1
    Debug=(bDrawDebug=false,bDraw2D=false,bDrawLocalAxes=false,DebugScale=25,DebugSegments=20,DebugThickness=0.25)
    DriverAxis=(X=-1,Y=0,Z=0)
    DriverItem=(Type=Bone,Name="upperarm_r")
    FalloffRegionScaleFactors=(PositiveWidth=0.2,NegativeWidth=0.2,PositiveHeight=0.2,NegativeHeight=0.3)
    FalloffSize=0.45
    FlipHeightScaling=false
    FlipWidthScaling=true
    OptionalParentItem=(Type=Bone,Name="spine_04")
    RotationOffset=(X=-20,Y=45,Z=30)
   )
   > &RigUnit_ModifyTransforms(
    ItemToModify=((Item=(Type=Bone,Name="dyn_strap_fr_r_MSTR1"),Transform=(Rotation=(X=0,Y=0,Z=0,W=1),Translation=(X=0,Y=-0.2,Z=2),Scale3D=(X=1,Y=0.2,Z=1))),(Item=(Type=Bone),Transform=(Rotation=(X=0,Y=0,Z=0,W=1),Translation=(X=0,Y=0,Z=0),Scale3D=(X=1,Y=1,Z=1))))
    ItemToModify.0.Transform.Rotation=&RigVMFunction_MathQuaternionFromEuler(Euler=(X=0,Y=0,Z=0), RotationOrder=XYZ).Result
    Weight=@n4.OutputParam
   )
   > n5 = &RigUnit_SphericalPoseReader(
    ActiveRegionScaleFactors=(PositiveWidth=0.5,NegativeWidth=0.5,PositiveHeight=0.5,NegativeHeight=0.5)
    ActiveRegionSize=0.2
    Debug=(bDrawDebug=false,DebugScale=25,DebugSegments=20,DebugThickness=0.25)
    DriverAxis=(X=-1,Y=0,Z=0)
    DriverItem=(Type=Bone,Name="upperarm_r")
    FalloffRegionScaleFactors=(PositiveWidth=0.4,NegativeWidth=0.2,PositiveHeight=0.1,NegativeHeight=0.1)
    FalloffSize=0.45
    OptionalParentItem=(Type=Bone,Name="spine_05")
    RotationOffset=(X=-75,Y=-60,Z=0)
   )
   > &RigUnit_ModifyTransforms(
    ItemToModify=((Item=(Type=Bone,Name="dyn_hood_01"),Transform=(Rotation=(X=0,Y=0,Z=0,W=1),Translation=(X=0,Y=-1,Z=0),Scale3D=(X=1,Y=1,Z=1))))
    ItemToModify.0.Transform.Rotation=&RigVMFunction_MathQuaternionFromEuler(Euler=(X=-10,Y=0,Z=0)).Result
    Weight=@n5.OutputParam
   )
   > n6 = &RigUnit_SphericalPoseReader(
    ActiveRegionScaleFactors=(PositiveWidth=0.2,NegativeWidth=0.2,PositiveHeight=0.5,NegativeHeight=0.5)
    ActiveRegionSize=0.2
    Debug=(bDrawDebug=false,DebugScale=25,DebugSegments=20,DebugThickness=0.25)
    DriverAxis=(X=-1,Y=0,Z=0)
    DriverItem=(Type=Bone,Name="upperarm_r")
    FalloffRegionScaleFactors=(PositiveWidth=0.4,NegativeWidth=0.2,PositiveHeight=0.1,NegativeHeight=0.1)
    FalloffSize=0.45
    OptionalParentItem=(Type=Bone,Name="spine_05")
    RotationOffset=(X=-45,Y=30,Z=-10)
   )
   > n7 = &RigUnit_ModifyTransforms(
    ItemToModify=((Item=(Type=Bone,Name="dyn_hood_01"),Transform=(Rotation=(X=0,Y=0,Z=0,W=1),Translation=(X=0,Y=-1,Z=0),Scale3D=(X=1,Y=1,Z=1))))
    ItemToModify.0.Transform.Rotation=&RigVMFunction_MathQuaternionFromEuler(Euler=(X=-10,Y=0,Z=0)).Result
    Weight=@n6.OutputParam
   )
   Branch_1_RigVMUnitNode = &RigVMFunction_ControlFlowBranch(ExecuteContext=@n7.ExecutePin, Condition=RigVMVariableNode() {variable=Backbling}.Value)
   Branch_RigVMUnitNode = &RigVMFunction_ControlFlowBranch(ExecuteContext=@Branch_1_RigVMUnitNode.False, Condition=RigVMVariableNode() {variable=Skydiving}.Value)
   n8 = &RigUnit_SphericalPoseReader(
    ActiveRegionScaleFactors=(PositiveWidth=0.4,NegativeWidth=0.2,PositiveHeight=0.7,NegativeHeight=0.5)
    ActiveRegionSize=0.2
    Debug=(bDrawDebug=false,DebugScale=25,DebugSegments=20,DebugThickness=0.25)
    DriverAxis=(X=1,Y=0,Z=0)
    DriverItem=(Type=Bone,Name="head")
    FalloffRegionScaleFactors=(PositiveWidth=0.35,NegativeWidth=0.1,PositiveHeight=0.4,NegativeHeight=0.2)
    FalloffSize=0.45
    OptionalParentItem=(Type=Bone,Name="neck_01")
    RotationOffset=(X=-90,Y=0,Z=45)
    ExecutePin=@Branch_RigVMUnitNode.False
   )
   > &RigUnit_ModifyTransforms(
    ItemToModify=((Item=(Type=Bone,Name="dyn_hood_01"),Transform=(Rotation=(X=0,Y=0,Z=0,W=1),Translation=(X=2,Y=-1.5,Z=-1.5),Scale3D=(X=1,Y=1,Z=1))),(Item=(Type=Bone),Transform=(Rotation=(X=0,Y=0,Z=0,W=1),Translation=(X=0,Y=0,Z=0),Scale3D=(X=1,Y=1,Z=1))))
    ItemToModify.0.Transform.Rotation=&RigVMFunction_MathQuaternionFromEuler(Euler=(X=5,Y=-8,Z=0), RotationOrder=XYZ).Result
    Weight=@n8.OutputParam
   )
   > &RigUnit_ModifyTransforms(
    ItemToModify=((Item=(Type=Bone,Name="dyn_hair_bk_01"),Transform=(Rotation=(X=0,Y=0,Z=0,W=1),Translation=(X=0,Y=0,Z=0),Scale3D=(X=1,Y=1,Z=1))),(Item=(Type=Bone),Transform=(Rotation=(X=0,Y=0,Z=0,W=1),Translation=(X=0,Y=0,Z=0),Scale3D=(X=1,Y=1,Z=1))))
    ItemToModify.0.Transform.Rotation=&RigVMFunction_MathQuaternionFromEuler(Euler=(X=0,Y=-5,Z=20), RotationOrder=XYZ).Result
    Weight=@n8.OutputParam
   )
   > &RigUnit_ModifyTransforms(
    ItemToModify=((Item=(Type=Bone,Name="dyn_hair_bk_mid_01"),Transform=(Rotation=(X=0,Y=0,Z=0,W=1),Translation=(X=0,Y=0,Z=0),Scale3D=(X=1,Y=1,Z=1))),(Item=(Type=Bone),Transform=(Rotation=(X=0,Y=0,Z=0,W=1),Translation=(X=0,Y=0,Z=0),Scale3D=(X=1,Y=1,Z=1))))
    ItemToModify.0.Transform.Rotation=&RigVMFunction_MathQuaternionFromEuler(Euler=(X=0,Y=-2,Z=20), RotationOrder=XYZ).Result
    Weight=@n8.OutputParam
   )
   n9 = &RigUnit_SphericalPoseReader(
    ActiveRegionScaleFactors=(PositiveWidth=1.5,NegativeWidth=0.2,PositiveHeight=0.7,NegativeHeight=0.7)
    ActiveRegionSize=0.2
    Debug=(bDrawDebug=True,DebugScale=25,DebugSegments=20,DebugThickness=0.25)
    DriverAxis=(X=1,Y=0,Z=0)
    DriverItem=(Type=Bone,Name="head")
    FalloffRegionScaleFactors=(PositiveWidth=1,NegativeWidth=0.1,PositiveHeight=0.5,NegativeHeight=0.5)
    FalloffSize=0.45
    OptionalParentItem=(Type=Bone,Name="neck_01")
    RotationOffset=(X=-90,Y=0,Z=45)
    ExecutePin=@Branch_RigVMUnitNode.True
   )
   > &RigUnit_ModifyTransforms(
    ItemToModify=((Item=(Type=Bone,Name="dyn_hood_01"),Transform=(Rotation=(X=0,Y=0,Z=0,W=1),Translation=(X=3,Y=-1,Z=0),Scale3D=(X=1,Y=1,Z=1))),(Item=(Type=Bone),Transform=(Rotation=(X=0,Y=0,Z=0,W=1),Translation=(X=0,Y=0,Z=0),Scale3D=(X=1,Y=1,Z=1))))
    ItemToModify.0.Transform.Rotation=&RigVMFunction_MathQuaternionFromEuler(Euler=(X=0,Y=0,Z=-2), RotationOrder=XYZ).Result
    Weight=@n9.OutputParam
   )
   AlphaInterp = &RigVMFunction_AlphaInterp(InterpSpeedDecreasing=7, InterpSpeedIncreasing=3, Scale=1, bClampResult=False, Value=RigVMVariableNode() {variable=Moving}.Value)
   &RigUnit_ModifyTransforms(
    ItemToModify=((Item=(Type=Bone,Name="dyn_hood_01"),Transform=(Rotation=(X=0,Y=0,Z=0,W=1),Translation=(X=2,Y=-3,Z=0),Scale3D=(X=1,Y=1,Z=1))),(Item=(Type=Bone),Transform=(Rotation=(X=0,Y=0,Z=0,W=1),Translation=(X=0,Y=0,Z=0),Scale3D=(X=1,Y=1,Z=1))))
    Weight=@AlphaInterp.Result
    ExecutePin=@Branch_1_RigVMUnitNode.True
    ItemToModify.0.Transform.Rotation=&RigVMFunction_MathQuaternionFromEuler(Euler=(X=0,Y=0,Z=-2), RotationOrder=XYZ).Result
   )
   > n10 = &RigUnit_SphericalPoseReader(
    ActiveRegionScaleFactors=(PositiveWidth=0.4,NegativeWidth=0.2,PositiveHeight=0.7,NegativeHeight=0.5)
    ActiveRegionSize=0.2
    Debug=(bDrawDebug=false,DebugScale=25,DebugSegments=20,DebugThickness=0.25)
    DriverAxis=(X=1,Y=0,Z=0)
    DriverItem=(Type=Bone,Name="head")
    FalloffRegionScaleFactors=(PositiveWidth=0.35,NegativeWidth=0.1,PositiveHeight=0.4,NegativeHeight=0.2)
    FalloffSize=0.45
    OptionalParentItem=(Type=Bone,Name="neck_01")
    RotationOffset=(X=-90,Y=0,Z=45)
   )
   > &RigUnit_ModifyTransforms(
    ItemToModify=((Item=(Type=Bone,Name="dyn_hood_01"),Transform=(Rotation=(X=0,Y=0,Z=0,W=1),Translation=(X=3,Y=0,Z=0),Scale3D=(X=1,Y=1,Z=1))),(Item=(Type=Bone),Transform=(Rotation=(X=0,Y=0,Z=0,W=1),Translation=(X=0,Y=0,Z=0),Scale3D=(X=1,Y=1,Z=1))))
    ItemToModify.0.Transform.Rotation=&RigVMFunction_MathQuaternionFromEuler(Euler=(X=8,Y=-8,Z=0), RotationOrder=XYZ).Result
    Weight=@n10.OutputParam
   )
   > &RigUnit_ModifyTransforms(
    ItemToModify=((Item=(Type=Bone,Name="dyn_hair_bk_01"),Transform=(Rotation=(X=0,Y=0,Z=0,W=1),Translation=(X=0,Y=0,Z=0),Scale3D=(X=1,Y=1,Z=1))),(Item=(Type=Bone),Transform=(Rotation=(X=0,Y=0,Z=0,W=1),Translation=(X=0,Y=0,Z=0),Scale3D=(X=1,Y=1,Z=1))))
    ItemToModify.0.Transform.Rotation=&RigVMFunction_MathQuaternionFromEuler(Euler=(X=0,Y=-5,Z=20), RotationOrder=XYZ).Result
    Weight=@n10.OutputParam
   )
   > &RigUnit_ModifyTransforms(
    ItemToModify=((Item=(Type=Bone,Name="dyn_hair_bk_mid_01"),Transform=(Rotation=(X=0,Y=0,Z=0,W=1),Translation=(X=0,Y=0,Z=0),Scale3D=(X=1,Y=1,Z=1))),(Item=(Type=Bone),Transform=(Rotation=(X=0,Y=0,Z=0,W=1),Translation=(X=0,Y=0,Z=0),Scale3D=(X=1,Y=1,Z=1))))
    ItemToModify.0.Transform.Rotation=&RigVMFunction_MathQuaternionFromEuler(Euler=(X=0,Y=-2,Z=20), RotationOrder=XYZ).Result
    Weight=@n10.OutputParam
   )
   n11 = &RigUnit_SphericalPoseReader(
    ActiveRegionScaleFactors=(PositiveWidth=0.6,NegativeWidth=0.6,PositiveHeight=1,NegativeHeight=1)
    ActiveRegionSize=0.15
    Debug=(bDrawDebug=True,bDraw2D=false,bDrawLocalAxes=true,DebugScale=25,DebugSegments=20,DebugThickness=0.25)
    DriverAxis=(X=1,Y=0,Z=0)
    DriverItem=(Type=Bone,Name="upperarm_l")
    ExecutePin=()
    FalloffRegionScaleFactors=(PositiveWidth=0.6,NegativeWidth=0.6,PositiveHeight=1,NegativeHeight=1)
    FalloffSize=0.45
    OptionalParentItem=(Type=Bone,Name="spine_04")
    RotationOffset=(X=40,Y=25,Z=-40)
   )
   > &RigUnit_ModifyTransforms(
    ItemToModify=((Item=(Type=Bone,Name="dyn_strap_fr_l_MSTR1"),Transform=(Rotation=(X=0,Y=0,Z=0,W=1),Translation=(X=0,Y=0,Z=-2),Scale3D=(X=1,Y=1,Z=1))))
    Weight=@n11.OutputParam
   )
  ]}
]
hierarchy [
 root {loc=<0 0 0>, rot=<0 0 0 1>}
 pelvis {parent=root, loc=<0 -0.7071 93.1191>, rot=<0.0127 0.707 -0.0127 -0.707>}
 spine_01 {parent=pelvis, loc=<6.7494 -0.6388 0>, rot=<0 0 0.034 -0.9994>}
 spine_02 {parent=spine_01, loc=<7.2518 -0.035 0>, rot=<0 0 0.0238 -0.9997>}
 spine_03 {parent=spine_02, loc=<6.0104 0 0>, rot=<0 0 0.0149 -0.9999>}
 spine_04 {parent=spine_03, loc=<8.1024 0 0>, rot=<0 0 -0.0132 -0.9999>}
 spine_05 {parent=spine_04, loc=<10.0146 0 0>, rot=<0 0 -0.0015 -1>}
 neck_01 {parent=spine_05, loc=<11.3138 0.1529 0>, rot=<0 0 -0.2003 -0.9797>}
 neck_02 {parent=neck_01, loc=<4.5627 0 0>, rot=<0 0 0.0504 -0.9987>}
 head {parent=neck_02, loc=<5.3914 0 0>, rot=<0 0 0.115 -0.9934>}
 dyn_hair_l_01 {parent=head, loc=<13.7151 2.8044 -3.7678>, rot=<-0.6819 -0.0926 -0.0732 -0.7218>}
 dyn_hair_l_02 {parent=dyn_hair_l_01, loc=<1.6019 0 0>, rot=<0.0771 -0.0772 -0.2377 -0.9652>}
 dyn_hair_l_03 {parent=dyn_hair_l_02, loc=<1.579 0 0>, rot=<0.0854 -0.0263 -0.1777 -0.98>}
 dyn_hair_l_04 {parent=dyn_hair_l_03, loc=<1.3444 0 0>, rot=<0.0218 -0.0016 -0.0406 -0.9989>}
 dyn_hair_c_01 {parent=head, loc=<12.8947 6.2386 -2.919>, rot=<0.492 0.4494 0.5565 0.4962>}
 dyn_hair_c_02 {parent=dyn_hair_c_01, loc=<2.5604 0 0>, rot=<0.3076 -0.0936 -0.1612 -0.9331>}
 dyn_hair_c_03 {parent=dyn_hair_c_02, loc=<2.1471 0 0>, rot=<0.0108 -0.0161 -0.2095 -0.9776>}
 dyn_hair_c_04 {parent=dyn_hair_c_03, loc=<1.8166 0 0>, rot=<0.0171 -0.0265 -0.0721 -0.9969>}
 dyn_hair_c_05 {parent=dyn_hair_c_04, loc=<1.7444 0 0>, rot=<0.0295 -0.0552 -0.0439 -0.9971>}
 dyn_hair_r_fr_01 {parent=head, loc=<13.1301 10.402 1.1933>, rot=<0.2603 -0.2644 0.877 0.3053>}
 dyn_hair_r_fr_02 {parent=dyn_hair_r_fr_01, loc=<2.0354 0 0>, rot=<0.0167 -0.0059 -0.0553 -0.9983>}
 dyn_hair_r_fr_03 {parent=dyn_hair_r_fr_02, loc=<2.158 0 0>, rot=<0.0288 -0.0133 -0.1677 -0.9853>}
 dyn_hair_r_fr_04 {parent=dyn_hair_r_fr_03, loc=<1.5493 0 0>, rot=<-0.0018 -0.0318 -0.1979 -0.9797>}
 dyn_hair_r_fr_05 {parent=dyn_hair_r_fr_04, loc=<2.0966 0 0>, rot=<-0.0116 -0.0278 -0.0699 -0.9971>}
 dyn_hair_r_fr_side_01 {parent=head, loc=<13.8752 8.7697 4.0916>, rot=<0.2424 -0.4963 0.789 0.2691>}
 dyn_hair_r_fr_side_02 {parent=dyn_hair_r_fr_side_01, loc=<2.5523 0.0612 0.0062>, rot=<0.0004 0.0019 -0.1543 -0.988>}
 dyn_hair_r_fr_side_03 {parent=dyn_hair_r_fr_side_02, loc=<2.2044 0.1121 0.0113>, rot=<0.0006 0.0065 -0.0974 -0.9952>}
 dyn_hair_r_fr_side_04 {parent=dyn_hair_r_fr_side_03, loc=<2.7652 0.3245 0.0327>, rot=<0.0017 0.0131 -0.1629 -0.9866>}
 dyn_hair_r_fr_top_01 {parent=head, loc=<14.861 7.3763 5.2043>, rot=<0.4248 -0.6569 0.5382 0.3136>}
 dyn_hair_r_fr_top_02 {parent=dyn_hair_r_fr_top_01, loc=<1.0175 0 0>, rot=<0.0158 -0.0027 0.0257 -0.9995>}
 dyn_hair_r_fr_top_03 {parent=dyn_hair_r_fr_top_02, loc=<1.2324 0 0>, rot=<0.0896 -0.0129 0.0669 -0.9936>}
 dyn_hair_r_fr_top_04 {parent=dyn_hair_r_fr_top_03, loc=<1.35 0 0>, rot=<0 0 0.08 -0.9968>}
 dyn_hair_r_fr_mid_01 {parent=head, loc=<11.8105 6.2874 7.0251>, rot=<0.5032 -0.6979 0.4007 0.3149>}
 dyn_hair_r_fr_mid_02 {parent=dyn_hair_r_fr_mid_01, loc=<1.4365 0 0>, rot=<0.0238 -0.0044 0.0782 -0.9966>}
 dyn_hair_r_fr_mid_03 {parent=dyn_hair_r_fr_mid_02, loc=<1.5764 0 0>, rot=<0 0 0.1112 -0.9938>}
 dyn_hair_r_fr_mid_04 {parent=dyn_hair_r_fr_mid_03, loc=<1.3846 0 0>, rot=<0 0 0 1>}
 dyn_hair_r_fr_bot_01 {parent=head, loc=<8.9616 5.4504 7.659>, rot=<0.1136 -0.6842 0.6838 -0.2267>}
 dyn_hair_r_fr_bot_02 {parent=dyn_hair_r_fr_bot_01, loc=<2.0225 -0.0873 0>, rot=<0 0 0.0486 -0.9988>}
 dyn_hair_r_fr_bot_03 {parent=dyn_hair_r_fr_bot_02, loc=<1.8391 0.0168 -0.0398>, rot=<0 0 0 1>}
 dyn_hair_r_side_top_01 {parent=head, loc=<14.3197 2.1154 7.3263>, rot=<0.5339 -0.6833 0.4635 0.1824>}
 dyn_hair_r_side_top_02 {parent=dyn_hair_r_side_top_01, loc=<1.1031 0 0>, rot=<0 0 0.0459 -0.9989>}
 dyn_hair_r_side_top_03 {parent=dyn_hair_r_side_top_02, loc=<0.867 0 0>, rot=<0 0 0 1>}
 dyn_hair_r_side_mid_01 {parent=head, loc=<10.3876 1.4396 7.9644>, rot=<0.533 -0.6642 0.4272 0.3038>}
 dyn_hair_r_side_mid_02 {parent=dyn_hair_r_side_mid_01, loc=<1.4733 -0.0976 0>, rot=<0 0 0.0681 -0.9977>}
 dyn_hair_r_side_mid_03 {parent=dyn_hair_r_side_mid_02, loc=<1.1878 0 0>, rot=<0 0 0 1>}
 dyn_hair_r_side_bot_01 {parent=head, loc=<10.7763 3.8478 7.2749>, rot=<0.2067 -0.7636 0.6044 -0.0948>}
 dyn_hair_r_side_bot_02 {parent=dyn_hair_r_side_bot_01, loc=<1.5963 0.0083 0.0139>, rot=<-0.0042 0.0213 -0.0128 -0.9997>}
 dyn_hair_r_side_bot_03 {parent=dyn_hair_r_side_bot_02, loc=<2.2886 0.0936 0.1607>, rot=<-0.0192 0.09 -0.0548 -0.9942>}
 dyn_hair_r_side_bot_04 {parent=dyn_hair_r_side_bot_03, loc=<1.2829 0 0>, rot=<0 0.1606 -0.0866 -0.9832>}
 dyn_hair_r_mid_top_01 {parent=head, loc=<16.4465 -0.6969 4.7511>, rot=<0.4544 -0.6975 0.4767 0.2826>}
 dyn_hair_r_mid_top_02 {parent=dyn_hair_r_mid_top_01, loc=<1.086 0 0>, rot=<0 0 0.171 -0.9853>}
 dyn_hair_r_mid_top_03 {parent=dyn_hair_r_mid_top_02, loc=<1.1531 0 0>, rot=<0 0 0 1>}
 dyn_hair_r_mid_01 {parent=head, loc=<14.6122 -1.7672 5.5526>, rot=<0.3111 -0.8344 0.3695 0.2656>}
 dyn_hair_r_mid_02 {parent=dyn_hair_r_mid_01, loc=<1.1839 0 0>, rot=<0 -0.0026 0.0067 -1>}
 dyn_hair_r_mid_03 {parent=dyn_hair_r_mid_02, loc=<1.4286 0 0>, rot=<0.0001 -0.008 0.0207 -0.9998>}
 dyn_hair_r_mid_04 {parent=dyn_hair_r_mid_03, loc=<0.9819 0 0>, rot=<0.0026 -0.0283 0.0589 -0.9979>}
 dyn_hair_r_mid_05 {parent=dyn_hair_r_mid_04, loc=<1.1312 0 0>, rot=<0.0179 -0.0711 0.0965 -0.9926>}
 dyn_hair_mid_bot_01 {parent=head, loc=<10.4085 -2.9206 7.2349>, rot=<-0.0828 -0.8393 0.4877 0.2257>}
 dyn_hair_mid_bot_02 {parent=dyn_hair_mid_bot_01, loc=<1.171 0 0>, rot=<0.0226 0.0632 -0.0345 -0.9971>}
 dyn_hair_mid_bot_03 {parent=dyn_hair_mid_bot_02, loc=<1.2158 0 0>, rot=<0.0316 0.1496 -0.0707 -0.9857>}
 dyn_hair_mid_bot_04 {parent=dyn_hair_mid_bot_03, loc=<1.3652 0 0>, rot=<0.0057 0.1274 -0.0477 -0.9907>}
 dyn_hair_mid_bot_05 {parent=dyn_hair_mid_bot_04, loc=<1.3778 0 0>, rot=<-0.0037 0.0696 -0.0081 -0.9975>}
 dyn_hair_bk_mid_top_01 {parent=head, loc=<13.7557 -4.8726 4.6006>, rot=<0.1131 -0.8712 0.3384 0.3371>}
 dyn_hair_bk_mid_top_02 {parent=dyn_hair_bk_mid_top_01, loc=<1.8708 -0.0153 0.0069>, rot=<0.0005 0.006 0.0133 -0.9999>}
 dyn_hair_bk_mid_top_03 {parent=dyn_hair_bk_mid_top_02, loc=<1.8376 -0.0337 0.0152>, rot=<0.0006 0.0182 0.0404 -0.999>}
 dyn_hair_bk_mid_top_04 {parent=dyn_hair_bk_mid_top_03, loc=<1.3359 0.0281 -0.0127>, rot=<-0.0084 0.0598 0.1298 -0.9897>}
 dyn_hair_bk_mid_01 {parent=head, loc=<8.4021 -5.5648 5.1496>, rot=<-0.1028 -0.9287 0.2054 0.2911>}
 dyn_hair_bk_mid_02 {parent=dyn_hair_bk_mid_01, loc=<2.2125 0.1053 0.1793>, rot=<0.0612 0.1541 -0.0503 -0.9849>}
 dyn_hair_bk_mid_03 {parent=dyn_hair_bk_mid_02, loc=<1.8033 0.0326 0.1275>, rot=<0.0442 0.1364 -0.0031 -0.9897>}
 dyn_hair_bk_mid_04 {parent=dyn_hair_bk_mid_03, loc=<1.7998 -0.0176 0.1685>, rot=<0.0146 0.2333 0.1387 -0.9623>}
 dyn_hair_bk_01 {parent=head, loc=<9.6406 -7.8336 0.5826>, rot=<-0.0843 -0.8811 0.0482 0.4629>}
 dyn_hair_bk_02 {parent=dyn_hair_bk_01, loc=<2.7565 0.0478 0.057>, rot=<0.0284 0.0993 -0.0786 -0.9915>}
 dyn_hair_bk_03 {parent=dyn_hair_bk_02, loc=<2.1265 0.0261 0.0509>, rot=<0.065 0.1453 -0.0628 -0.9853>}
 dyn_hair_bk_04 {parent=dyn_hair_bk_03, loc=<2.8271 0.0395 0.3274>, rot=<0.0623 0.1165 -0.0067 -0.9912>}
 FACIAL_C_FacialRoot {parent=head, loc=<2.7957 0.4816 0>, rot=<0.5021 -0.4979 -0.5021 -0.4979>}
 FACIAL_C_UnderChin {parent=FACIAL_C_FacialRoot, loc=<0 7.9848 3.748>, rot=<0.5 0 0 -0.866>}
 FACIAL_L_UnderChin {parent=FACIAL_C_FacialRoot, loc=<2.6901 7.0296 2.9205>, rot=<0.4025 -0.355 0.0823 -0.8398>}
 FACIAL_R_UnderChin {parent=FACIAL_C_FacialRoot, loc=<-2.6898 7.0295 2.9213>, rot=<0.4025 0.355 -0.0823 -0.8398>}
 FACIAL_C_Jaw {parent=FACIAL_C_FacialRoot, loc=<0 2.0958 1.5277>, rot=<0.342 0 0 -0.9397>}
 FACIAL_C_Jawline {parent=FACIAL_C_Jaw, loc=<0 2.0985 7.7296>, rot=<0.2588 0 0 -0.9659>}
 FACIAL_L_ChinSide {parent=FACIAL_C_Jaw, loc=<2.5987 -0.0675 7.0635>, rot=<-0.2207 -0.3739 0.05 -0.8994>}
 FACIAL_R_ChinSide {parent=FACIAL_C_Jaw, loc=<-2.5987 -0.0675 7.0635>, rot=<-0.2207 0.3739 -0.05 -0.8994>}
 FACIAL_C_Chin {parent=FACIAL_C_Jaw, loc=<0 0.5097 7.7929>, rot=<-0.2588 0 0 -0.9659>}
 FACIAL_L_Jawline {parent=FACIAL_C_FacialRoot, loc=<3.8145 5.2187 4.1745>, rot=<0.2092 0.1578 -0.0547 -0.9635>}
 FACIAL_R_Jawline {parent=FACIAL_C_FacialRoot, loc=<-3.8145 5.2187 4.1745>, rot=<0.2092 -0.1578 0.0547 -0.9635>}
 dyn_zipper {parent=neck_02, loc=<-0.5002 5.5893 0>, rot=<-0.1217 -0.9926 0 0>}
 clavicle_l {parent=spine_05, loc=<4.3575 -0.1466 -2.3223>, rot=<0.0465 -0.7289 -0.009 -0.6829>}
 upperarm_l {parent=clavicle_l, loc=<15.5111 0 0>, rot=<0.0226 -0.3234 -0.0088 -0.9459>}
 upperarm_twist_01_l {parent=upperarm_l, loc=<0.0041 0 0>, rot=<0 0 0 1>}
 deform_upperarm_fr_base_l1 {parent=upperarm_twist_01_l, loc=<-0.3691 7.0506 -2.765>, rot=<0 0 0 1>}
 deform_upperarm_fr_l {parent=deform_upperarm_fr_base_l1, loc=<0 0 0>, rot=<0 0 0 1>}
 clavicle_r {parent=spine_05, loc=<4.3575 -0.1466 2.3223>, rot=<0.7289 0.0465 -0.6829 0.009>}
 upperarm_r {parent=clavicle_r, loc=<-15.5111 0 0>, rot=<0.0226 -0.3234 -0.0088 -0.9459>}
 upperarm_twist_01_r {parent=upperarm_r, loc=<-0.0041 0 0>, rot=<0 0 0 -1>}
 deform_upperarm_fr_base_r1 {parent=upperarm_twist_01_r, loc=<0.3689 -7.0506 2.7652>, rot=<0 0 0 1>}
 deform_upperarm_fr_r {parent=deform_upperarm_fr_base_r1, loc=<0 0 0>, rot=<0 0 0 1>}
 dyn_hood_01 {parent=spine_05, loc=<21.3083 -8.0889 0.0837>, rot=<0.1531 -0.9881 -0.001 0.0133>}
 dyn_hood_02 {parent=dyn_hood_01, loc=<2.2286 0 0>, rot=<0 0 0 1>}
 dyn_hood_03 {parent=dyn_hood_02, loc=<3.9968 0 0>, rot=<0 0 0 1>}
 dyn_hood_04 {parent=dyn_hood_03, loc=<3.8674 0 0>, rot=<0 0 0 1>}
 dyn_hood_05 {parent=dyn_hood_04, loc=<3.709 0 0>, rot=<0 0 0 1>}
 dyn_hood_06 {parent=dyn_hood_05, loc=<3.6913 0 0>, rot=<0 0 0 1>}
 deform_pec_aim_base_r1 {parent=spine_04, loc=<7.8727 12.6388 2.4123>, rot=<-0.4602 -0.0102 0.7936 0.3979>}
 deform_pec_aim_r {parent=deform_pec_aim_base_r1, loc=<0 0 0>, rot=<0 0 0 1>}
 deform_pec_base_r {parent=deform_pec_aim_r, loc=<-10.7662 -0.0006 -0.0007>, rot=<0 0 0 1>}
 deform_pec_r {parent=deform_pec_base_r, loc=<0 0 0>, rot=<0 0 0 1>}
 dyn_strap_fr_r_MSTR1 {parent=spine_04, loc=<13.1797 13.3781 8.5212>, rot=<-0.0673 -0.9972 -0.005 -0.0333>}
 dyn_strap_fr_r_01 {parent=dyn_strap_fr_r_MSTR1, loc=<0 0 0>, rot=<0 0 0 1>}
 deform_pec_aim_base_l1 {parent=spine_04, loc=<7.8732 12.6388 -2.4123>, rot=<0.0102 -0.4602 0.3979 -0.7936>}
 deform_pec_aim_l {parent=deform_pec_aim_base_l1, loc=<0 0 0>, rot=<0 0 0 1>}
 deform_pec_base_l {parent=deform_pec_aim_l, loc=<10.766 0 0>, rot=<0 0 0 1>}
 deform_pec_l {parent=deform_pec_base_l, loc=<0.055 0 0>, rot=<0 0 0 1>}
 dyn_strap_fr_l_MSTR1 {parent=spine_04, loc=<13.1797 13.3781 -8.5212>, rot=<-0.0673 -0.9972 0.005 0.0333>}
 dyn_strap_fr_l_01 {parent=dyn_strap_fr_l_MSTR1, loc=<0 0 0>, rot=<0 0 0 1>}
]
variables [
 Backbling {type=bool, default=False}
 Moving {type=double, default=0}
 Skydiving {type=bool, default=False}
]
