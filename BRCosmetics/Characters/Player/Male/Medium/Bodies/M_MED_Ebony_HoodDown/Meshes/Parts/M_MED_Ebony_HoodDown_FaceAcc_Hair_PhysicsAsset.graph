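# Physics asset for the hair part of the male medium Ebony HoodDown player cosmetic, providing bodies for dynamic hair bones.
# Sixteen strand chains of small capsules, mostly on the right side and back, hang from a head body built from a sphere and capsule.
# Root links to the head have free swing and twist.
# Links along each strand have tight per-strand swing limits, stiffest on the left strand, and almost no twist.
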PhysicsAsset /BRCosmetics/Characters/Player/Male/Medium/Bodies/M_MED_Ebony_HoodDown/Meshes/Parts/M_MED_Ebony_HoodDown_FaceAcc_Hair_PhysicsAsset
bodies {
 dyn_hair_bk_01 [{shapeType=Capsule, length=0.5, radius=0.5, shapeName=dyn_hair_bk_01_capsule}]
 dyn_hair_bk_02 [{shapeType=Capsule, length=2, radius=0.6, shapeName=dyn_hair_bk_02_capsule, center=<1 0 0>, rotation=<rot 90 0 0>}]
 dyn_hair_bk_03 [{shapeType=Capsule, length=2, radius=0.6, shapeName=dyn_hair_bk_03_capsule, center=<1 0 0>, rotation=<rot 90 0 0>}]
 dyn_hair_bk_04 [{shapeType=Capsule, length=2, radius=0.6, shapeName=dyn_hair_bk_04_capsule, center=<1 0 0>, rotation=<rot 90 0 0>}]
 dyn_hair_bk_mid_01 [{shapeType=Capsule, length=0.5, radius=0.5, shapeName=dyn_hair_bk_mid_01_capsule}]
 dyn_hair_bk_mid_02 [{shapeType=Capsule, length=2, radius=0.6, shapeName=dyn_hair_bk_mid_02_capsule, center=<1 0 0>, rotation=<rot 90 0 0>}]
 dyn_hair_bk_mid_03 [{shapeType=Capsule, length=2, radius=0.6, shapeName=dyn_hair_bk_mid_03_capsule, center=<1 0 0>, rotation=<rot 90 0 0>}]
 dyn_hair_bk_mid_04 [{shapeType=Capsule, length=2, radius=0.6, shapeName=dyn_hair_bk_mid_04_capsule, center=<1 0 0>, rotation=<rot 90 0 0>}]
 dyn_hair_bk_mid_top_01 [{shapeType=Capsule, length=0.5, radius=0.5, shapeName=dyn_hair_bk_mid_top_01_capsule}]
 dyn_hair_bk_mid_top_02 [{shapeType=Capsule, length=2, radius=0.6, shapeName=dyn_hair_bk_mid_top_02_capsule, center=<1 0 0>, rotation=<rot 90 0 0>}]
 dyn_hair_bk_mid_top_03 [{shapeType=Capsule, length=2, radius=0.6, shapeName=dyn_hair_bk_mid_top_03_capsule, center=<1 0 0>, rotation=<rot 90 0 0>}]
 dyn_hair_bk_mid_top_04 [{shapeType=Capsule, length=2, radius=0.6, shapeName=dyn_hair_bk_mid_top_04_capsule, center=<1 0 0>, rotation=<rot 90 0 0>}]
 dyn_hair_c_01 [{shapeType=Capsule, length=0.5, radius=0.5, shapeName=dyn_hair_c_01_capsule}]
 dyn_hair_c_02 [{shapeType=Capsule, length=2, radius=0.6, shapeName=dyn_hair_c_02_capsule, center=<2.5 0 0>, rotation=<rot 90 0 0>}]
 dyn_hair_c_03 [{shapeType=Capsule, length=2, radius=0.6, shapeName=dyn_hair_c_03_capsule, center=<2.5 0 0>, rotation=<rot 90 0 0>}]
 dyn_hair_c_04 [{shapeType=Capsule, length=2, radius=0.6, shapeName=dyn_hair_c_04_capsule, center=<2.5 0 0>, rotation=<rot 90 0 0>}]
 dyn_hair_c_05 [{shapeType=Capsule, length=2, radius=0.6, shapeName=dyn_hair_c_05_capsule, center=<2.5 0 0>, rotation=<rot 90 0 0>}]
 dyn_hair_l_01 [{shapeType=Capsule, length=0.5, radius=0.5, shapeName=dyn_hair_l_01_capsule}]
 dyn_hair_l_02 [{shapeType=Capsule, length=2, radius=0.6, shapeName=dyn_hair_l_02_capsule, center=<2 0 0>, rotation=<rot 90 0 0>}]
 dyn_hair_l_03 [{shapeType=Capsule, length=2, radius=0.6, shapeName=dyn_hair_l_03_capsule, center=<2 0 0>, rotation=<rot 90 0 0>}]
 dyn_hair_l_04 [{shapeType=Capsule, length=2, radius=0.6, shapeName=dyn_hair_l_04_capsule, center=<2 0 0>, rotation=<rot 90 0 0>}]
 dyn_hair_mid_bot_01 [{shapeType=Capsule, length=0.5, radius=0.5, shapeName=dyn_hair_mid_bot_01_capsule}]
 dyn_hair_mid_bot_02 [{shapeType=Capsule, length=2, radius=0.6, shapeName=dyn_hair_mid_bot_02_capsule, center=<2 0 0>, rotation=<rot 90 0 0>}]
 dyn_hair_mid_bot_03 [{shapeType=Capsule, length=2, radius=0.6, shapeName=dyn_hair_mid_bot_03_capsule, center=<2 0 0>, rotation=<rot 90 0 0>}]
 dyn_hair_mid_bot_04 [{shapeType=Capsule, length=2, radius=0.6, shapeName=dyn_hair_mid_bot_04_capsule, center=<2 0 0>, rotation=<rot 90 0 0>}]
 dyn_hair_mid_bot_05 [{shapeType=Capsule, length=2, radius=0.6, shapeName=dyn_hair_mid_bot_05_capsule, center=<2 0 0>, rotation=<rot 90 0 0>}]
 dyn_hair_r_fr_01 [{shapeType=Capsule, length=0.5, radius=0.5, shapeName=dyn_hair_r_fr_01_capsule}]
 dyn_hair_r_fr_02 [{shapeType=Capsule, length=2, radius=0.6, shapeName=dyn_hair_r_fr_02_capsule, center=<2.5 0 0>, rotation=<rot 90 0 0>}]
 dyn_hair_r_fr_03 [{shapeType=Capsule, length=2, radius=0.6, shapeName=dyn_hair_r_fr_03_capsule, center=<2.5 0 0>, rotation=<rot 90 0 0>}]
 dyn_hair_r_fr_04 [{shapeType=Capsule, length=2, radius=0.6, shapeName=dyn_hair_r_fr_04_capsule, center=<2.5 0 0>, rotation=<rot 90 0 0>}]
 dyn_hair_r_fr_05 [{shapeType=Capsule, length=2, radius=0.6, shapeName=dyn_hair_r_fr_05_capsule, center=<2.5 0 0>, rotation=<rot 90 0 0>}]
 dyn_hair_r_fr_bot_01 [{shapeType=Capsule, length=0.5, radius=0.5, shapeName=dyn_hair_r_fr_bot_01_capsule}]
 dyn_hair_r_fr_bot_02 [{shapeType=Capsule, length=2, radius=0.6, shapeName=dyn_hair_r_fr_bot_02_capsule, center=<0.2835 0 0>, rotation=<rot 90 0 0>}]
 dyn_hair_r_fr_bot_03 [{shapeType=Capsule, length=2, radius=0.6, shapeName=dyn_hair_r_fr_bot_03_capsule, center=<1 0 0>, rotation=<rot 90 0 0>}]
 dyn_hair_r_fr_mid_01 [{shapeType=Capsule, length=0.5, radius=0.5, shapeName=dyn_hair_r_fr_mid_01_capsule}]
 dyn_hair_r_fr_mid_02 [{shapeType=Capsule, length=2, radius=0.6, shapeName=dyn_hair_r_fr_mid_02_capsule, center=<1 0 0>, rotation=<rot 90 0 0>}]
 dyn_hair_r_fr_mid_03 [{shapeType=Capsule, length=2, radius=0.6, shapeName=dyn_hair_r_fr_mid_03_capsule, center=<1 0 0>, rotation=<rot 90 0 0>}]
 dyn_hair_r_fr_mid_04 [{shapeType=Capsule, length=2, radius=0.6, shapeName=dyn_hair_r_fr_mid_04_capsule, center=<1 0 0>, rotation=<rot 90 0 0>}]
 dyn_hair_r_fr_side_01 [{shapeType=Capsule, length=0.5, radius=0.5, shapeName=dyn_hair_r_fr_side_01_capsule}]
 dyn_hair_r_fr_side_02 [{shapeType=Capsule, length=2, radius=0.6, shapeName=dyn_hair_r_fr_side_02_capsule, center=<2.5 0 0>, rotation=<rot 90 0 0>}]
 dyn_hair_r_fr_side_03 [{shapeType=Capsule, length=2, radius=0.6, shapeName=dyn_hair_r_fr_side_03_capsule, center=<2.5 0 0>, rotation=<rot 90 0 0>}]
 dyn_hair_r_fr_side_04 [{shapeType=Capsule, length=2, radius=0.6, shapeName=dyn_hair_r_fr_side_04_capsule, center=<2.5 0 0>, rotation=<rot 90 0 0>}]
 dyn_hair_r_fr_top_01 [{shapeType=Capsule, length=0.5, radius=0.5, shapeName=dyn_hair_r_fr_top_01_capsule}]
 dyn_hair_r_fr_top_02 [{shapeType=Capsule, length=2, radius=0.6, shapeName=dyn_hair_r_fr_top_02_capsule, center=<1 0 0>, rotation=<rot 90 0 0>}]
 dyn_hair_r_fr_top_03 [{shapeType=Capsule, length=2, radius=0.6, shapeName=dyn_hair_r_fr_top_03_capsule, center=<1 0 0>, rotation=<rot 90 0 0>}]
 dyn_hair_r_fr_top_04 [{shapeType=Capsule, length=2, radius=0.6, shapeName=dyn_hair_r_fr_top_04_capsule, center=<1 0 0>, rotation=<rot 90 0 0>}]
 dyn_hair_r_mid_01 [{shapeType=Capsule, length=0.5, radius=0.5, shapeName=dyn_hair_r_mid_01_capsule}]
 dyn_hair_r_mid_02 [{shapeType=Capsule, length=2, radius=0.6, shapeName=dyn_hair_r_mid_02_capsule, rotation=<rot 90 0 0>}]
 dyn_hair_r_mid_03 [{shapeType=Capsule, length=2, radius=0.6, shapeName=dyn_hair_r_mid_03_capsule, rotation=<rot 90 0 0>}]
 dyn_hair_r_mid_04 [{shapeType=Capsule, length=2, radius=0.6, shapeName=dyn_hair_r_mid_04_capsule, rotation=<rot 90 0 0>}]
 dyn_hair_r_mid_05 [{shapeType=Capsule, length=2, radius=0.6, shapeName=dyn_hair_r_mid_05_capsule, rotation=<rot 90 0 0>}]
 dyn_hair_r_mid_top_01 [{shapeType=Capsule, length=0.5, radius=0.5, shapeName=dyn_hair_r_mid_top_01_capsule}]
 dyn_hair_r_mid_top_02 [{shapeType=Capsule, length=2, radius=0.6, shapeName=dyn_hair_r_mid_top_02_capsule, center=<1 0 0>, rotation=<rot 90 0 0>}]
 dyn_hair_r_mid_top_03 [{shapeType=Capsule, length=2, radius=0.6, shapeName=dyn_hair_r_mid_top_03_capsule, center=<1 0 0>, rotation=<rot 90 0 0>}]
 dyn_hair_r_side_bot_01 [{shapeType=Capsule, length=0.5, radius=0.5, shapeName=dyn_hair_r_side_bot_01_capsule}]
 dyn_hair_r_side_bot_02 [{shapeType=Capsule, length=2, radius=0.6, shapeName=dyn_hair_r_side_bot_02_capsule, center=<1 0 0>, rotation=<rot 90 0 0>}]
 dyn_hair_r_side_bot_03 [{shapeType=Capsule, length=2, radius=0.6, shapeName=dyn_hair_r_side_bot_03_capsule, center=<1 0 0>, rotation=<rot 90 0 0>}]
 dyn_hair_r_side_bot_04 [{shapeType=Capsule, length=2, radius=0.6, shapeName=dyn_hair_r_side_bot_04_capsule, center=<1 0 0>, rotation=<rot 90 0 0>}]
 dyn_hair_r_side_mid_01 [{shapeType=Capsule, length=0.5, radius=0.5, shapeName=dyn_hair_r_side_mid_01_capsule}]
 dyn_hair_r_side_mid_02 [{shapeType=Capsule, length=2, radius=0.6, shapeName=dyn_hair_r_side_mid_02_capsule, center=<1 0 0>, rotation=<rot 90 0 0>}]
 dyn_hair_r_side_mid_03 [{shapeType=Capsule, length=2, radius=0.6, shapeName=dyn_hair_r_side_mid_03_capsule, center=<1 0 0>, rotation=<rot 90 0 0>}]
 dyn_hair_r_side_top_01 [{shapeType=Capsule, length=0.5, radius=0.5, shapeName=dyn_hair_r_side_top_01_capsule}]
 dyn_hair_r_side_top_02 [{shapeType=Capsule, length=2, radius=0.6, shapeName=dyn_hair_r_side_top_02_capsule, center=<2 0 0>, rotation=<rot 90 0 0>}]
 dyn_hair_r_side_top_03 [{shapeType=Capsule, length=2, radius=0.6, shapeName=dyn_hair_r_side_top_03_capsule, center=<2 0 0>, rotation=<rot 90 0 0>}]
 head [
  {shapeType=Sphere, radius=6.5041, shapeName=head_sphere, center=<5.0677 0 1.1294>}
  {
   shapeType=Capsule
   length=6.6335
   radius=5.7102
   shapeName=head_capsule
   center=<5.3979 5.2443 2.0764>
   rotation=<rot 0.6376 81.7878 -87.4537>
  }
 ]
}
constraints [
 {bone1Name=dyn_hair_bk_02, bone2Name=dyn_hair_bk_01, swing1LimitDegrees=10, swing2LimitDegrees=10, twistLimitDegrees=1}
 {bone1Name=dyn_hair_bk_03, bone2Name=dyn_hair_bk_02, swing1LimitDegrees=10, swing2LimitDegrees=10, twistLimitDegrees=1}
 {bone1Name=dyn_hair_bk_04, bone2Name=dyn_hair_bk_03, swing1LimitDegrees=10, swing2LimitDegrees=10, twistLimitDegrees=1}
 {bone1Name=dyn_hair_bk_mid_02, bone2Name=dyn_hair_bk_mid_01, swing1LimitDegrees=10, swing2LimitDegrees=10, twistLimitDegrees=1}
 {bone1Name=dyn_hair_bk_mid_03, bone2Name=dyn_hair_bk_mid_02, swing1LimitDegrees=10, swing2LimitDegrees=10, twistLimitDegrees=1}
 {bone1Name=dyn_hair_bk_mid_04, bone2Name=dyn_hair_bk_mid_03, swing1LimitDegrees=10, swing2LimitDegrees=10, twistLimitDegrees=1}
 {bone1Name=dyn_hair_bk_mid_top_02, bone2Name=dyn_hair_bk_mid_top_01, swing1LimitDegrees=10, swing2LimitDegrees=10, twistLimitDegrees=1}
 {bone1Name=dyn_hair_bk_mid_top_03, bone2Name=dyn_hair_bk_mid_top_02, swing1LimitDegrees=10, swing2LimitDegrees=10, twistLimitDegrees=1}
 {bone1Name=dyn_hair_bk_mid_top_04, bone2Name=dyn_hair_bk_mid_top_03, swing1LimitDegrees=10, swing2LimitDegrees=10, twistLimitDegrees=1}
 {bone1Name=dyn_hair_c_02, bone2Name=dyn_hair_c_01, swing1LimitDegrees=20, swing2LimitDegrees=20, twistLimitDegrees=1}
 {bone1Name=dyn_hair_c_03, bone2Name=dyn_hair_c_02, swing1LimitDegrees=20, swing2LimitDegrees=20, twistLimitDegrees=1}
 {bone1Name=dyn_hair_c_04, bone2Name=dyn_hair_c_03, swing1LimitDegrees=20, swing2LimitDegrees=20, twistLimitDegrees=1}
 {bone1Name=dyn_hair_c_05, bone2Name=dyn_hair_c_04, swing1LimitDegrees=20, swing2LimitDegrees=20, twistLimitDegrees=1}
 {bone1Name=dyn_hair_l_02, bone2Name=dyn_hair_l_01, swing1LimitDegrees=5, swing2LimitDegrees=1, twistLimitDegrees=1}
 {bone1Name=dyn_hair_l_03, bone2Name=dyn_hair_l_02, swing1LimitDegrees=5, swing2LimitDegrees=1, twistLimitDegrees=1}
 {bone1Name=dyn_hair_l_04, bone2Name=dyn_hair_l_03, swing1LimitDegrees=5, swing2LimitDegrees=1, twistLimitDegrees=1}
 {bone1Name=dyn_hair_mid_bot_02, bone2Name=dyn_hair_mid_bot_01, swing1LimitDegrees=20, swing2LimitDegrees=20, twistLimitDegrees=1}
 {bone1Name=dyn_hair_mid_bot_03, bone2Name=dyn_hair_mid_bot_02, swing1LimitDegrees=20, swing2LimitDegrees=20, twistLimitDegrees=1}
 {bone1Name=dyn_hair_mid_bot_04, bone2Name=dyn_hair_mid_bot_03, swing1LimitDegrees=20, swing2LimitDegrees=20, twistLimitDegrees=1}
 {bone1Name=dyn_hair_mid_bot_05, bone2Name=dyn_hair_mid_bot_04, swing1LimitDegrees=20, swing2LimitDegrees=20, twistLimitDegrees=1}
 {bone1Name=dyn_hair_r_fr_02, bone2Name=dyn_hair_r_fr_01, swing1LimitDegrees=20, swing2LimitDegrees=20, twistLimitDegrees=1}
 {bone1Name=dyn_hair_r_fr_03, bone2Name=dyn_hair_r_fr_02, swing1LimitDegrees=20, swing2LimitDegrees=20, twistLimitDegrees=1}
 {bone1Name=dyn_hair_r_fr_04, bone2Name=dyn_hair_r_fr_03, swing1LimitDegrees=20, swing2LimitDegrees=20, twistLimitDegrees=1}
 {bone1Name=dyn_hair_r_fr_05, bone2Name=dyn_hair_r_fr_04, swing1LimitDegrees=20, swing2LimitDegrees=20, twistLimitDegrees=1}
 {bone1Name=dyn_hair_r_fr_bot_02, bone2Name=dyn_hair_r_fr_bot_01, swing1LimitDegrees=20, swing2LimitDegrees=20, twistLimitDegrees=1}
 {bone1Name=dyn_hair_r_fr_bot_03, bone2Name=dyn_hair_r_fr_bot_02, swing1LimitDegrees=20, swing2LimitDegrees=20, twistLimitDegrees=1}
 {bone1Name=dyn_hair_r_fr_mid_02, bone2Name=dyn_hair_r_fr_mid_01, swing1LimitDegrees=10, swing2LimitDegrees=10, twistLimitDegrees=1}
 {bone1Name=dyn_hair_r_fr_mid_03, bone2Name=dyn_hair_r_fr_mid_02, swing1LimitDegrees=10, swing2LimitDegrees=10, twistLimitDegrees=1}
 {bone1Name=dyn_hair_r_fr_mid_04, bone2Name=dyn_hair_r_fr_mid_03, swing1LimitDegrees=10, swing2LimitDegrees=10, twistLimitDegrees=1}
 {bone1Name=dyn_hair_r_fr_side_02, bone2Name=dyn_hair_r_fr_side_01, swing1LimitDegrees=20, swing2LimitDegrees=20, twistLimitDegrees=1}
 {bone1Name=dyn_hair_r_fr_side_03, bone2Name=dyn_hair_r_fr_side_02, swing1LimitDegrees=20, swing2LimitDegrees=20, twistLimitDegrees=1}
 {bone1Name=dyn_hair_r_fr_side_04, bone2Name=dyn_hair_r_fr_side_03, swing1LimitDegrees=20, swing2LimitDegrees=20, twistLimitDegrees=1}
 {bone1Name=dyn_hair_r_fr_top_02, bone2Name=dyn_hair_r_fr_top_01, swing1LimitDegrees=15, swing2LimitDegrees=15, twistLimitDegrees=1}
 {bone1Name=dyn_hair_r_fr_top_03, bone2Name=dyn_hair_r_fr_top_02, swing1LimitDegrees=15, swing2LimitDegrees=15, twistLimitDegrees=1}
 {bone1Name=dyn_hair_r_fr_top_04, bone2Name=dyn_hair_r_fr_top_03, swing1LimitDegrees=15, swing2LimitDegrees=15, twistLimitDegrees=1}
 {bone1Name=dyn_hair_r_mid_02, bone2Name=dyn_hair_r_mid_01, swing1LimitDegrees=10, swing2LimitDegrees=10, twistLimitDegrees=1}
 {bone1Name=dyn_hair_r_mid_03, bone2Name=dyn_hair_r_mid_02, swing1LimitDegrees=10, swing2LimitDegrees=10, twistLimitDegrees=1}
 {bone1Name=dyn_hair_r_mid_04, bone2Name=dyn_hair_r_mid_03, swing1LimitDegrees=10, swing2LimitDegrees=10, twistLimitDegrees=1}
 {bone1Name=dyn_hair_r_mid_05, bone2Name=dyn_hair_r_mid_04, swing1LimitDegrees=10, swing2LimitDegrees=10, twistLimitDegrees=1}
 {bone1Name=dyn_hair_r_mid_top_02, bone2Name=dyn_hair_r_mid_top_01, swing1LimitDegrees=10, swing2LimitDegrees=10, twistLimitDegrees=1}
 {bone1Name=dyn_hair_r_mid_top_03, bone2Name=dyn_hair_r_mid_top_02, swing1LimitDegrees=10, swing2LimitDegrees=10, twistLimitDegrees=1}
 {bone1Name=dyn_hair_r_side_bot_02, bone2Name=dyn_hair_r_side_bot_01, swing1LimitDegrees=20, swing2LimitDegrees=20, twistLimitDegrees=1}
 {bone1Name=dyn_hair_r_side_bot_03, bone2Name=dyn_hair_r_side_bot_02, swing1LimitDegrees=20, swing2LimitDegrees=20, twistLimitDegrees=1}
 {bone1Name=dyn_hair_r_side_bot_04, bone2Name=dyn_hair_r_side_bot_03, swing1LimitDegrees=20, swing2LimitDegrees=20, twistLimitDegrees=1}
 {bone1Name=dyn_hair_r_side_mid_02, bone2Name=dyn_hair_r_side_mid_01, swing1LimitDegrees=10, swing2LimitDegrees=10, twistLimitDegrees=1}
 {bone1Name=dyn_hair_r_side_mid_03, bone2Name=dyn_hair_r_side_mid_02, swing1LimitDegrees=10, swing2LimitDegrees=10, twistLimitDegrees=1}
 {bone1Name=dyn_hair_r_side_top_02, bone2Name=dyn_hair_r_side_top_01, swing1LimitDegrees=10, swing2LimitDegrees=10, twistLimitDegrees=1}
 {bone1Name=dyn_hair_r_side_top_03, bone2Name=dyn_hair_r_side_top_02, swing1LimitDegrees=10, swing2LimitDegrees=10, twistLimitDegrees=1}
 {
  bone1Name=dyn_hair_l_01
  bone2Name=head
  swing1LimitDegrees=45
  swing1Motion=Free
  swing2LimitDegrees=45
  swing2Motion=Free
  twistLimitDegrees=45
  twistMotion=Free
 }
 {
  bone1Name=dyn_hair_c_01
  bone2Name=head
  swing1LimitDegrees=45
  swing1Motion=Free
  swing2LimitDegrees=45
  swing2Motion=Free
  twistLimitDegrees=45
  twistMotion=Free
 }
 {
  bone1Name=dyn_hair_r_fr_01
  bone2Name=head
  swing1LimitDegrees=45
  swing1Motion=Free
  swing2LimitDegrees=45
  swing2Motion=Free
  twistLimitDegrees=45
  twistMotion=Free
 }
 {
  bone1Name=dyn_hair_r_fr_side_01
  bone2Name=head
  swing1LimitDegrees=45
  swing1Motion=Free
  swing2LimitDegrees=45
  swing2Motion=Free
  twistLimitDegrees=45
  twistMotion=Free
 }
 {
  bone1Name=dyn_hair_r_fr_top_01
  bone2Name=head
  swing1LimitDegrees=45
  swing1Motion=Free
  swing2LimitDegrees=45
  swing2Motion=Free
  twistLimitDegrees=45
  twistMotion=Free
 }
 {
  bone1Name=dyn_hair_r_fr_mid_01
  bone2Name=head
  swing1LimitDegrees=45
  swing1Motion=Free
  swing2LimitDegrees=45
  swing2Motion=Free
  twistLimitDegrees=45
  twistMotion=Free
 }
 {
  bone1Name=dyn_hair_r_fr_bot_01
  bone2Name=head
  swing1LimitDegrees=45
  swing1Motion=Free
  swing2LimitDegrees=45
  swing2Motion=Free
  twistLimitDegrees=45
  twistMotion=Free
 }
 {
  bone1Name=dyn_hair_r_side_top_01
  bone2Name=head
  swing1LimitDegrees=45
  swing1Motion=Free
  swing2LimitDegrees=45
  swing2Motion=Free
  twistLimitDegrees=45
  twistMotion=Free
 }
 {
  bone1Name=dyn_hair_r_side_mid_01
  bone2Name=head
  swing1LimitDegrees=45
  swing1Motion=Free
  swing2LimitDegrees=45
  swing2Motion=Free
  twistLimitDegrees=45
  twistMotion=Free
 }
 {
  bone1Name=dyn_hair_r_side_bot_01
  bone2Name=head
  swing1LimitDegrees=45
  swing1Motion=Free
  swing2LimitDegrees=45
  swing2Motion=Free
  twistLimitDegrees=45
  twistMotion=Free
 }
 {
  bone1Name=dyn_hair_r_mid_top_01
  bone2Name=head
  swing1LimitDegrees=45
  swing1Motion=Free
  swing2LimitDegrees=45
  swing2Motion=Free
  twistLimitDegrees=45
  twistMotion=Free
 }
 {
  bone1Name=dyn_hair_r_mid_01
  bone2Name=head
  swing1LimitDegrees=45
  swing1Motion=Free
  swing2LimitDegrees=45
  swing2Motion=Free
  twistLimitDegrees=45
  twistMotion=Free
 }
 {
  bone1Name=dyn_hair_mid_bot_01
  bone2Name=head
  swing1LimitDegrees=45
  swing1Motion=Free
  swing2LimitDegrees=45
  swing2Motion=Free
  twistLimitDegrees=45
  twistMotion=Free
 }
 {
  bone1Name=dyn_hair_bk_mid_top_01
  bone2Name=head
  swing1LimitDegrees=45
  swing1Motion=Free
  swing2LimitDegrees=45
  swing2Motion=Free
  twistLimitDegrees=45
  twistMotion=Free
 }
 {
  bone1Name=dyn_hair_bk_mid_01
  bone2Name=head
  swing1LimitDegrees=45
  swing1Motion=Free
  swing2LimitDegrees=45
  swing2Motion=Free
  twistLimitDegrees=45
  twistMotion=Free
 }
 {
  bone1Name=dyn_hair_bk_01
  bone2Name=head
  swing1LimitDegrees=45
  swing1Motion=Free
  swing2LimitDegrees=45
  swing2Motion=Free
  twistLimitDegrees=45
  twistMotion=Free
 }
]
