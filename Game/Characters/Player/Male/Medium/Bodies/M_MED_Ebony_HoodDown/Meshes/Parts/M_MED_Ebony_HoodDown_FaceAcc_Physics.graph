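# Physics asset for the face accessory mesh of the medium male Ebony hood-down player body, setting up the hood's bone chain for physics.
# Six capsule bodies, dyn_hood_01 to 06, link in series from spine_05 through tight swing and very narrow twist limits.
# Larger spine_04 and spine_05 capsules plus a neck_01 capsule give upper-body volumes, joined to one another by constraints with free swing and twist.
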PhysicsAsset /Game/Characters/Player/Male/Medium/Bodies/M_MED_Ebony_HoodDown/Meshes/Parts/M_MED_Ebony_HoodDown_FaceAcc_Physics
bodies {
 dyn_hood_01 [{shapeType=Capsule, length=5, radius=2, shapeName=dyn_hood_01_capsule, center=<5 0 0>}]
 dyn_hood_02 [{shapeType=Capsule, length=4.2091, radius=2.7909, shapeName=dyn_hood_02_capsule, center=<5 0 0>}]
 dyn_hood_03 [{shapeType=Capsule, length=4.2091, radius=2.7909, shapeName=dyn_hood_03_capsule, center=<5 0 0>}]
 dyn_hood_04 [{shapeType=Capsule, length=4.2091, radius=2.7909, shapeName=dyn_hood_04_capsule, center=<5 0 0>}]
 dyn_hood_05 [{shapeType=Capsule, length=4.2091, radius=2.7909, shapeName=dyn_hood_05_capsule, center=<5 0 0>}]
 dyn_hood_06 [{shapeType=Capsule, length=4.2091, radius=2.7909, shapeName=dyn_hood_06_capsule, center=<5 0 0>}]
 neck_01 [{shapeType=Capsule, length=5, radius=5, shapeName=neck_01_capsule}]
 spine_04 [{shapeType=Capsule, length=11.0565, radius=9.7131, shapeName=spine_04_capsule, center=<4.7588 0 0>}]
 spine_05 [{shapeType=Capsule, length=21.0339, radius=10.6361, shapeName=spine_05_capsule, center=<4.9607 0.3386 0>}]
}
constraints [
 {bone1Name=dyn_hood_02, bone2Name=dyn_hood_01, swing1LimitDegrees=10, swing2LimitDegrees=15, twistLimitDegrees=2}
 {bone1Name=dyn_hood_03, bone2Name=dyn_hood_02, swing1LimitDegrees=10, swing2LimitDegrees=15, twistLimitDegrees=2}
 {bone1Name=dyn_hood_04, bone2Name=dyn_hood_03, swing1LimitDegrees=10, swing2LimitDegrees=15, twistLimitDegrees=2}
 {bone1Name=dyn_hood_05, bone2Name=dyn_hood_04, swing1LimitDegrees=10, swing2LimitDegrees=15, twistLimitDegrees=2}
 {bone1Name=dyn_hood_06, bone2Name=dyn_hood_05, swing1LimitDegrees=10, swing2LimitDegrees=15, twistLimitDegrees=2}
 {bone1Name=dyn_hood_01, bone2Name=spine_05, swing1LimitDegrees=10, swing2LimitDegrees=15, twistLimitDegrees=1}
 {
  bone1Name=spine_05
  bone2Name=spine_04
  swing1LimitDegrees=45
  swing1Motion=Free
  swing2LimitDegrees=45
  swing2Motion=Free
  twistLimitDegrees=45
  twistMotion=Free
 }
 {
  bone1Name=neck_01
  bone2Name=spine_05
  swing1LimitDegrees=45
  swing1Motion=Free
  swing2LimitDegrees=45
  swing2Motion=Free
  twistLimitDegrees=45
  twistMotion=Free
 }
]
